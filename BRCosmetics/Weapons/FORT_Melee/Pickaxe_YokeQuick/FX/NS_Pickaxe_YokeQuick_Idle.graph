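# Idle glow effect for the YokeQuick pickaxe cosmetic, drawing warm gold glow sprites, a mesh and short-lived god rays.
# One emitter packs four spawn groups into sprite and mesh renderers via VisibilityTag: bursts persist, while rate-spawned rays animate and expire through KillGodRay.
# Particles follow a sampled socket with per-group offsets and orientations.
# Time-of-day and system fade scale brightness, with platform-quality and front-end adjustments.
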NiagaraSystem /BRCosmetics/Weapons/FORT_Melee/Pickaxe_YokeQuick/FX/NS_Pickaxe_YokeQuick_Idle
dependencies {
 usedDataInterfaces [
  &NiagaraDataInterfacePlatformSet
  &NiagaraDataInterfaceMeshRendererInfo
  &NiagaraDataInterfaceSpriteRendererInfo
  &NiagaraDataInterfaceCurve
  &NiagaraDataInterfaceSocketReader
 ]
 usedDynamicInputs [
  &Add_Float
  &AddVectorToPosition
  &FNTimeOfDayScalar
 ]
 usedModules [
  &self:SystemSpawnScript.SystemScriptSource.SystemScriptGraph.NiagaraNodeAssignment_1.SetVariables_#1
  &SystemState
  &self:SystemSpawnScript.SystemScriptSource.SystemScriptGraph.NiagaraNodeAssignment_0.SetVariables_#2
  &self:SystemSpawnScript.SystemScriptSource.SystemScriptGraph.NiagaraNodeAssignment_2.SetVariables_#3
 ]
}
emitters {Stacked_Plane_Glow001 {
  emitterUpdateScript [
   &EmitterState(
    Inactive Response=::NewEnumerator1
    Life Cycle Mode=::NewEnumerator1
    Loop Behavior=::NewEnumerator1
    Loop Duration Mode=::NewEnumerator1
    UseLoopDelay=false
   )
   &SpawnBurst_Instantaneous(Use Spawn Probability=false, UseLoopCountLimit=false) {module=SpawnBurst_Instantaneous002}
   &SpawnBurst_Instantaneous(Spawn Group=1, Use Spawn Probability=false, UseLoopCountLimit=false) {module=SpawnBurst_Instantaneous001}
   &SpawnRate(
    SpawnGroup=2
    SpawnRate=&FN_PickFloatDependOnPlatform(TargetPlatform=DataInterface() {propertyValues {Platforms {qualityLevelMask=-29}}}, TargetPlatformActive=2, TargetPlatformNoActive=5)
    Use Spawn Probability=false
   )
   &SpawnBurst_Instantaneous(Spawn Count=3, Spawn Group=3, Use Spawn Probability=false, UseLoopCountLimit=false) {module=SpawnBurst_Instantaneous003}
  ]
  particleSpawnScript [
   &FN_SetSpawnGroup(Number of Spawn Groups:NiagaraInt32=/Niagara/DynamicInputs/Helpers/MaskInt(
     Mask Source:NiagaraInt32=0
     Mask Parameter:NiagaraInt32=$Emitter.SpawnGroup
     Mask:NiagaraInt32=3
     Masked Int:NiagaraInt32=3
     Passthrough Int:NiagaraInt32=20
    ))
   &InitializeParticle(
    Color=<rgba 20 12.0833 1 1>
    Lifetime=5
    Sprite Rotation Angle=90
    Sprite Rotation Mode=::NewEnumerator0
    Sprite Size=<84.5325 44.5275>
    Sprite Size Mode=::NewEnumerator4
    UsePositionOffset=false
    Write Material Random=true
   )
   &SubUVAnimation(
    End Frame Range Override=3
    Evaluation Type=::NewEnumerator0
    Output Alternate Renderer Binding=true
    "Sprite Renderer "=DataInterface() {propertyValues {SpriteRenderer=&self:Stacked_Plane_Glow001_0.NiagaraSpriteRendererProperties_0}}
    SubUV Animation Mode:ENiagara_SubUVLookupModeV2=::NewEnumerator2
    UseEndFrame=true
    UseStartFrame=true
   )
   /SetVariables(Particles.VisibilityTag=$Emitter.SpawnGroup) {isSetParameters=true}
   /SetVariables(Particles.GodRayLifetime:NiagaraFloat=&RandomRangeFloat(Minimum=1.5, Maximum=2)) {isSetParameters=true}
   /SetVariables(Particles.DynamicParameterAlt:Vector4f=&MakeVector4(X=&Multiply_FloatByInt(Float=0.1, Integer=&SpawnGroup()), Y=1, Z=1, W=1)) {isSetParameters=true}
   /SetVariables(Particles.DynamicParameterAlt1:Vector4f=&MakeVector4(
     X=&MergeHelper_FloatSwitch(0=1.5, 1=1.5, 2=3, 3=?, Index=&SpawnGroup(), Num=::NewEnumerator1)
     Y=10
     Z=0.2
     W=&MergeHelper_FloatSwitch(0=0.55, 1=0.55, 2=1, 3=?, Index=&SpawnGroup(), Num=::NewEnumerator1)
    )) {isSetParameters=true}
  ]
  particleUpdateScript [
   &ParticleState(Kill Particles When Lifetime Has Elapsed=false)
   /SetVariables(Particles.GodRayAge:NiagaraFloat=$Particles.Age) {isSetParameters=true}
   &self:KillGodRay(MaskParameter:NiagaraInt32=$Particles.VisibilityTag, MaskValue=2) {module=KillGodRay}
   &SocketLocation(
    SamplePosOffset=true
    Sampled Orientation Coordinate Space:ENiagaraCoordinateSpace=::Local
    Sampled Position Offset=&MergeHelper_VectorSwitch(
     0=<-4 0 2.0283>
     1=<-18 0 -5.0381>
     2=<-4 0 2.0283>
     3=&MergeHelper_VectorSwitch(
      0=<-13 18.5 2.4933>
      1=<-13 -18.5 2.4933>
      2=<-24.9465 0 2.4933>
      3=?
      Index=&SpawnGroup()
      Num=::NewEnumerator1
     )
     Index=$Particles.VisibilityTag
     Num=::NewEnumerator2
    )
    Socket Reader=$System.Socket Reader
    Socket Sampling Mode=::NewEnumerator3
    VelocitySpeedLimit=false
   )
   &FN_Sprite_Attributes(
    Color=&MakeLinearColorFromVectorAndFloat(
     "Vector (RGB)"=&Multiply_VectorByFloat(Vector=&MakeVectorFromLinearColorRGB(Color=<rgba 3 1.8125 0.15 0.5>), Float=0.55)
     "Float (Alpha)"=&Multiply_Float(A=$System.TODScalar, B=$System.SystemFade)
    )
    Create Color Attribute=true
    Write to Module Namespace=true
   )
   &FN_Mesh_Attributes(
    Color=&Multiply_LinearColorByFloat(Linear Color=<rgba 5 2.9167 0 1>, Float=&Multiply_Float(A=$System.TODScalar, B=$System.SystemFade))
    Create Color Attribute=true
    Create Scale Attribute=true
    Scale=<1.43 1.585 1.45>
    Write to Module Namespace=true
   )
   &FN_Sprite_Attributes(
    Color=&Multiply_LinearColorByFloat(
     Linear Color=<rgba 1 0.6042 0.05 1>
     Float=&FloatFromCurve(
      FloatCurve=DataInterface() {propertyValues {Curve {keys [<at 0 0 RCIM_Cubic>,<at 0.3 1 RCIM_Cubic>,<at 1 0 RCIM_Cubic>]}}}
      CurveIndex=$Particles.GodRayNormalizedAge
      Scale Curve=&Multiply_Float(A=$System.TODScalar, B=$System.SystemFade)
     )
    )
    Create Color Attribute=true
    Create Sprite Rotation Attribute=true
    Create SpriteSize Attribute=true
    SpriteRotation:NiagaraFloat=&RandomRangeFloat(Minimum=-65, Maximum=65)
    SpriteSize=&Multiply_Vector2D_ByFloat(
     Float=&FloatFromCurve(
      FloatCurve=DataInterface() {propertyValues {Curve {keys [
          <at -0.0017 0.0417 RCIM_Cubic>
          {arriveTangent=1.1486, interpMode=RCIM_Cubic, leaveTangent=1.1486, tangentMode=RCTM_User, time=0.0317, value=1.2332}
          <at 0.57 1.2568>
          {arriveTangent=-0.1439, interpMode=RCIM_Cubic, leaveTangent=-0.1439, tangentMode=RCTM_User, time=0.9845, value=1.0006}
          {interpMode=RCIM_Cubic, tangentMode=RCTM_User, time=1.0011, value=0.002}
         ]}}}
      CurveIndex=$Particles.GodRayNormalizedAge
     )
     Vector2D=&RandomRangeVector2D(Minimum=<32.5 39>, Maximum=<45.5 45.5>)
    )
    Write to Module Namespace=true
   ) {module=FN_Sprite_Attributes002}
   &FN_Sprite_Attributes(
    Camera Offset=3
    Color=&MakeLinearColorFromVectorAndFloat(
     "Vector (RGB)"=&MakeVectorFromLinearColorRGB(Color=<rgba 3 1.8125 0.15 1>)
     "Float (Alpha)"=&Multiply_Float(
      A=&Multiply_Float(
       A=&FN_FrontEndFloatModifier(Gameplay=1, Front End=0.4)
       B=&Multiply_Float(A=$System.TODScalar, B=$System.SystemFade)
      )
      B=&FN_PickFloatDependOnPlatform(TargetPlatform=DataInterface() {propertyValues {Platforms {qualityLevelMask=-29}}}, TargetPlatformActive=0.05, TargetPlatformNoActive=0.8)
     )
    )
    Create Camera Offset Attribute=true
    Create Color Attribute=true
    Create SpriteSize Attribute=true
    SpriteSize=&MergeHelper_Vector2Switch(0=<13 38>, 1=<13 38>, 2=<13 53>, 3=?, 4=?, Index=&SpawnGroup(), Num=::NewEnumerator1)
    Write to Module Namespace=true
   ) {module=FN_Sprite_Attributes003}
   &SolveForcesAndVelocity(Clamp Velocity=false, Limit Acceleration=false)
   &AlignSpriteToMeshOrientation(
    Mesh Orientation Relative Sprite Alignment Vector=&MaskVector(
     Mask Source=1
     Mask Parameter=$Particles.VisibilityTag
     Mask=0
     Masked Vector=<0 1 0>
     Passthrough Vector=&MaskVector(
      Mask Source=1
      Mask Parameter=$Particles.VisibilityTag
      Mask=3
      Masked Vector=&MergeHelper_VectorSwitch(0=<-1 0.04 0.05>, 1=<-1 -0.04 0.05>, 2=<0 1 0>, 3=?, Index=&SpawnGroup(), Num=::NewEnumerator1)
      Passthrough Vector=<-1 0 0>
     )
    )
    Mesh Orientation Relative Sprite Facing Vector:Vector3f=&MaskVector(
     Mask Source=1
     Mask Parameter=$Particles.VisibilityTag
     Mask=0
     Masked Vector=<0 0 1>
     Passthrough Vector=&MaskVector(
      Mask Source=1
      Mask Parameter=$Particles.VisibilityTag
      Mask=3
      Masked Vector=&MergeHelper_VectorSwitch(0=<0 1 0>, 1=<0 1 0>, 2=<0 0 1>, 3=?, Index=&SpawnGroup(), Num=::NewEnumerator1)
      Passthrough Vector=<0 0 -1>
     )
    )
    Orientation Quaternion:Quat4f=$Output.SocketLocation.SampledRotation
    SetAlignment=true
    SetFacing=true
   )
  ]
  rendererClasses [&NiagaraSpriteRendererProperties, &NiagaraMeshRendererProperties]
  renderers [
   {rendererClass=&NiagaraSpriteRendererProperties}
   {rendererClass=&NiagaraMeshRendererProperties, rendererIndex=1}
   {rendererClass=&NiagaraSpriteRendererProperties, rendererIndex=2}
   {rendererClass=&NiagaraSpriteRendererProperties, rendererIndex=3}
  ]
 }}
summary {
 systemName=NS_Pickaxe_YokeQuick_Idle
 emitters [{
   emitterName=Stacked_Plane_Glow001
   rendererClasses [&NiagaraSpriteRendererProperties, &NiagaraMeshRendererProperties]
  }]
 userVariables [
  User.bIsFrontEnd {type=NiagaraBool}
  User.bIsFrontEndPreview {type=NiagaraBool}
  User.PhysicalSurface {type=NiagaraInt32}
 ]
}
