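# Physics asset for the front-end (menu display) mesh of the YokeQuick pickaxe, a cosmetic melee weapon.
# It defines capsule bodies on two dynamic bones: a tiny one on the base bone and a larger, offset one on the main bone.
# A single constraint joins main to base, allowing a wider swing on the second axis while the first swing and twist have very tight limits, suggesting a mostly one-plane dangle.
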PhysicsAsset /BRCosmetics/Weapons/FORT_Melee/Pickaxe_YokeQuick/Meshes/YokeQuick_Axe_FrontEnd_Physics
bodies {
 dyn_yokeQuick_base [{shapeType=Capsule, length=0.505, radius=0.505, shapeName=dyn_yokeQuick_base_capsule, rotation=<rot 0 -90 0>}]
 dyn_yokeQuick_main [{shapeType=Capsule, length=20, radius=15, shapeName=dyn_yokeQuick_main_capsule, center=<15 0 4.15>, rotation=<rot 0 0 -90>}]
}
constraints [{bone1Name=dyn_yokeQuick_main, bone2Name=dyn_yokeQuick_base, swing1LimitDegrees=1, swing2LimitDegrees=35, twistLimitDegrees=1}]
