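# Physics asset for the YokeQuick pickaxe cosmetic, defining capsule bodies and joints for a dangling or swinging dyn_yokeQuick part.
# The large dyn_yokeQuick_main capsule joins a tiny dyn_yokeQuick_base capsule through a hinge-like constraint, with twist and one swing axis locked and the other swing left at limited motion.
# Thigh and calf capsules on both legs, joined by free angular constraints, suggest collision proxies for the character's legs.
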PhysicsAsset /BRCosmetics/Weapons/FORT_Melee/Pickaxe_YokeQuick/Meshes/YokeQuick_Axe_Physics
bodies {
 calf_l [{shapeType=Capsule, length=30.5504, radius=7.4496, shapeName=thigh_l_capsule, center=<-26 0 -4.2974>, rotation=<rot 90 180 0>}]
 calf_r [{shapeType=Capsule, length=30.5504, radius=8.9496, shapeName=thigh_r_capsule, center=<26 0 -1.3537>, rotation=<rot 90 0 0>}]
 dyn_yokeQuick_base [{shapeType=Capsule, length=0.505, radius=0.505, shapeName=dyn_yokeQuick_base_capsule, rotation=<rot 0 -90 0>}]
 dyn_yokeQuick_main [{
   shapeType=Capsule
   length=5.4755
   radius=23.6131
   shapeName=dyn_yokeQuick_main_capsule
   center=<9.4723 0 17.6988>
   rotation=<rot 0.0002 -90 -90>
  }]
 thigh_l [{shapeType=Capsule, length=35.212, radius=8.9975, shapeName=thigh_l_capsule, center=<-22.1907 -0.6066 0.3009>, rotation=<rot 90 180 0>}]
 thigh_r [{shapeType=Capsule, length=35.212, radius=8.9975, shapeName=thigh_r_capsule, center=<22.1907 0.6066 -0.3009>, rotation=<rot 90 0 0>}]
}
constraints [
 {
  bone1Name=calf_l
  bone2Name=thigh_l
  swing1LimitDegrees=45
  swing1Motion=Free
  swing2LimitDegrees=45
  swing2Motion=Free
  twistLimitDegrees=45
  twistMotion=Free
 }
 {
  bone1Name=calf_r
  bone2Name=thigh_r
  swing1LimitDegrees=45
  swing1Motion=Free
  swing2LimitDegrees=45
  swing2Motion=Free
  twistLimitDegrees=45
  twistMotion=Free
 }
 {
  bone1Name=dyn_yokeQuick_main
  bone2Name=dyn_yokeQuick_base
  swing1LimitDegrees=1
  swing1Motion=Locked
  swing2LimitDegrees=35
  twistLimitDegrees=1
  twistMotion=Locked
 }
]
