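# Animation Blueprint for the YokeQuick pickaxe, simulating secondary motion on its dangling and hinged parts.
# Event graph caches the attach parent's root, pelvis and leg socket transforms, and selects rigid-body alphas and lock sim-space settings for solo front-end preview versus general use.
# AnimGraph runs world-space replace nodes on matching body bones, two RigidBody passes (standard and FrontEnd physics assets), then chained AnimDynamics for the box hinge and two locks.
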AnimBlueprint /BRCosmetics/Weapons/FORT_Melee/Pickaxe_YokeQuick/Meshes/YokeQuick_Axe_AnimBP
target_skeleton=&/YokeQuick_Axe_Skeleton
graphs [
 AnimGraph {nodes [
   n1 = Anim Dynamics - Bone: dyn_boxOpenClose(ComponentPose=Local To Component(LocalPose=Use cached pose 'Rigid Body Simulate Pose'()), Alpha=1, ExternalForce="0,0,0") {
    angularDampingOverride=0.9999
    angularSpringConstant=2000
    bGravityOverrideInSimSpace=true
    bOverrideAngularDamping=true
    bUseGravityOverride=true
    simulationSpace=BoneRelative
    boundBone {boneName=dyn_boxOpenClose}
    gravityOverride {x=-980}
    physicsBodyDefinitions [{
      boundBone {boneName=dyn_boxOpenClose}
      boxExtents=<15 15 15>
      constraintSetup {angularLimitsMax=<0 5 0>, angularLimitsMin=<0 -2 0>, angularTarget=<0 0 0>}
      localJointOffset=<-3 0 0>
     }]
    relativeSpaceBone {boneName=dyn_yokeQuick_main}
    simSpaceSettings {maxAngularAcceleration=10, maxAngularVelocity=2, simSpaceAngularAlpha=1}
   }
   n2 = Anim Dynamics - Bone: dyn_down_phyLock!(
    ComponentPose=@n1
    Alpha=1
    SimSpaceSettings:Anim Phys Sim Space Settings Structure=(SimSpaceAngularAlpha=0,MaxAngularVelocity=10000,MaxAngularAcceleration=10000,ExternalAngularVelocity=(X=0,Y=0,Z=0))
    ExternalForce="0,0,0"
   ) -> (Pose) {
    angularDampingOverride=0.995
    angularSpringConstant=5000
    bGravityOverrideInSimSpace=true
    bOverrideAngularDamping=true
    bUseGravityOverride=true
    simulationSpace=BoneRelative
    boundBone {boneName=dyn_down_phyLock}
    componentAppliedLinearAccClamp=<0 0 0>
    gravityOverride {z=680}
    physicsBodyDefinitions [{
      boundBone {boneName=dyn_down_phyLock}
      boxExtents=<10 10 10>
      constraintSetup {angularLimitsMax=<0 0 0>, angularLimitsMin=<0 -60 0>, angularTarget=<0 0 0>}
      localJointOffset=<4 0 0>
     }]
    relativeSpaceBone {boneName=dyn_boxOpenClose}
   }
   n3 = Anim Dynamics - Bone: dyn_up_phyLock!(
    ComponentPose=@n2
    Alpha=1
    SimSpaceSettings:Anim Phys Sim Space Settings Structure=(SimSpaceAngularAlpha=0,MaxAngularVelocity=10000,MaxAngularAcceleration=10000,ExternalAngularVelocity=(X=0,Y=0,Z=0))
    ExternalForce="0,0,0"
   ) -> (Pose) {
    angularDampingOverride=0.999
    angularSpringConstant=5000
    bGravityOverrideInSimSpace=true
    bOverrideAngularDamping=true
    bUseGravityOverride=true
    simulationSpace=BoneRelative
    boundBone {boneName=dyn_up_phyLock}
    componentAppliedLinearAccClamp=<0 0 0>
    gravityOverride {z=680}
    physicsBodyDefinitions [{
      boundBone {boneName=dyn_up_phyLock}
      boxExtents=<10 10 10>
      constraintSetup {angularLimitsMax=<0 0 0>, angularLimitsMin=<0 -60 0>, angularTarget=<0 0 0>}
      localJointOffset=<2 0 0>
     }]
    relativeSpaceBone {boneName=dyn_boxOpenClose}
   }
   Output Pose(Result=Component To Local(ComponentPose=@n3)) {layerGroup=None}
   AnimGraphNode_SubInput_0 = Input Pose()
   n4 = "Transform (Modify) Bone - Bone: body_root"(ComponentPose=Local To Component(LocalPose=@AnimGraphNode_SubInput_0), Alpha=1, Translation="0,0,0", Rotation="0,0,0", Scale="1,1,1") {
    rotationMode=BMM_Replace
    rotationSpace=BCS_WorldSpace
    scaleSpace=BCS_BoneSpace
    translationMode=BMM_Replace
    translationSpace=BCS_WorldSpace
    alphaBoolBlend {blendOption=Linear}
    alphaScaleBiasClamp {bInterpResult=true, interpSpeedDecreasing=1, interpSpeedIncreasing=1, inRange {max=0, min=-40}, outRange {max=0, min=1}}
    boneToModify {boneName=body_root}
   }
   n5 = "Transform (Modify) Bone - Bone: pelvis"(ComponentPose=@n4, Alpha=1, Translation="0,0,0", Rotation="0,0,0", Scale="1,1,1") {
    rotationMode=BMM_Replace
    rotationSpace=BCS_WorldSpace
    scaleSpace=BCS_BoneSpace
    translationMode=BMM_Replace
    translationSpace=BCS_WorldSpace
    alphaBoolBlend {blendOption=Linear}
    alphaScaleBiasClamp {inRange {max=0, min=-40}, outRange {max=0, min=1}}
    boneToModify {boneName=pelvis}
   }
   n6 = "Transform (Modify) Bone - Bone: thigh_l"(ComponentPose=@n5, Alpha=1, Translation="0,0,0", Rotation="0,0,0", Scale="1,1,1") {
    rotationMode=BMM_Replace
    rotationSpace=BCS_WorldSpace
    scaleSpace=BCS_BoneSpace
    translationMode=BMM_Replace
    translationSpace=BCS_WorldSpace
    alphaBoolBlend {blendOption=Linear}
    alphaScaleBiasClamp {inRange {max=0, min=-40}, outRange {max=0, min=1}}
    boneToModify {boneName=thigh_l}
   }
   n7 = "Transform (Modify) Bone - Bone: thigh_r"(ComponentPose=@n6, Alpha=1, Translation="0,0,0", Rotation="0,0,0", Scale="1,1,1") {
    rotationMode=BMM_Replace
    rotationSpace=BCS_WorldSpace
    scaleSpace=BCS_BoneSpace
    translationMode=BMM_Replace
    translationSpace=BCS_WorldSpace
    alphaBoolBlend {blendOption=Linear}
    alphaScaleBiasClamp {inRange {max=0, min=-40}, outRange {max=0, min=1}}
    boneToModify {boneName=thigh_r}
   }
   n8 = "Transform (Modify) Bone - Bone: calf_l"(ComponentPose=@n7, Alpha=1, Translation="0,0,0", Rotation="0,0,0", Scale="1,1,1") {
    rotationMode=BMM_Replace
    rotationSpace=BCS_WorldSpace
    scaleSpace=BCS_BoneSpace
    translationMode=BMM_Replace
    translationSpace=BCS_WorldSpace
    alphaBoolBlend {blendOption=Linear}
    alphaScaleBiasClamp {inRange {max=0, min=-40}, outRange {max=0, min=1}}
    boneToModify {boneName=calf_l}
   }
   n9 = "Transform (Modify) Bone - Bone: calf_r"(ComponentPose=@n8, Alpha=1, Translation="0,0,0", Rotation="0,0,0", Scale="1,1,1") {
    rotationMode=BMM_Replace
    rotationSpace=BCS_WorldSpace
    scaleSpace=BCS_BoneSpace
    translationMode=BMM_Replace
    translationSpace=BCS_WorldSpace
    alphaBoolBlend {blendOption=Linear}
    alphaScaleBiasClamp {inRange {max=0, min=-40}, outRange {max=0, min=1}}
    boneToModify {boneName=calf_r}
   }
   n10 = RigidBody(ComponentPose=@n9) {
    bClampLinearTranslationLimitToRefPose=true
    bOverrideWorldGravity=true
    overridePhysicsAsset=&/YokeQuick_Axe_Physics
    simulationSpace=BaseBoneSpace
    baseBoneRef {boneName=b_AR_Root}
    componentAppliedLinearAccClamp=<3500 3500 3500>
    componentLinearAccScale=<0.6 0.6 0.6>
    componentLinearVelScale=<0.8 0.8 0.8>
    overrideWorldGravity {z=-980}
    simSpaceSettings {dampingAlpha=0}
   }
   n11 = RigidBody(ComponentPose=@n10) {
    bClampLinearTranslationLimitToRefPose=true
    bOverrideWorldGravity=true
    overridePhysicsAsset=&/YokeQuick_Axe_FrontEnd_Physics
    simulationSpace=BaseBoneSpace
    baseBoneRef {boneName=b_AR_Root}
    componentAppliedLinearAccClamp=<1500 3500 1500>
    componentLinearAccScale=<0.6 0.8 0.6>
    componentLinearVelScale=<1 2 1>
    overrideWorldGravity {z=-980}
    simSpaceSettings {dampingAlpha=0}
   }
   Save cached pose 'Rigid Body Simulate Pose'(Pose=Component To Local(ComponentPose=@n11)) {cacheName=Rigid Body Simulate Pose}
  ]}
 EventGraph {nodes [
   Event BlueprintUpdateAnimation()
   n1 = GetPartSkeletalMeshComponent()
   > Is Valid(InputObject=@n1)
   n2 = GetAttachParent!(self:Scene Component Object Reference=@n1) -> (ReturnValue:Scene Component Object Reference)
   > Is Valid(InputObject=@n2)
   > Set Body Mesh(Body Mesh:Scene Component Object Reference=@n2) -> (Output_Get:Scene Component Object Reference)
   n3 = Get bIsOnPawn() -> (bIsOnPawn:bool)
   > Set IsPreviewingItemAloneInFN(IsPreviewingItemAloneInFN:bool=AND Boolean(A=Get bIsInFrontEnd(), B=NOT Boolean(A=@n3))) -> (Output_Get:bool)
   > n4 = Sequence()
   > n5 = Sequence()
   n6 = Get IsPreviewingItemAloneInFN() -> (IsPreviewingItemAloneInFN:bool)
   > Set RigidBodyAlphaForDisplayAlone(RigidBodyAlphaForDisplayAlone:double=SelectFloat(A=1, B=0, bPickA=@n6)) -> (Output_Get:double)
   > Set RigidBodyAlphaForGeneral(RigidBodyAlphaForGeneral:double=SelectFloat(A=0, B=1, bPickA=@n6)) -> (Output_Get:double)
   n7 = Branch(execute=@n5.then_1, Condition=Get IsPreviewingItemAloneInFN())
   > Set SimSpaceSettingForLowerLock(SimSpaceSettingForLowerLock:Anim Phys Sim Space Settings Structure=Get SimSpaceSettingForLowerLock_FN_Alone() -> (SimSpaceSettingForLowerLock_FN_Alone:Anim Phys Sim Space Settings Structure)) -> (Output_Get:Anim Phys Sim Space Settings Structure)
   > Set SimSpaceSettingForUpperLock(SimSpaceSettingForUpperLock:Anim Phys Sim Space Settings Structure=Get SimSpaceSettingForUpperLock_FN_Alone() -> (SimSpaceSettingForUpperLock_FN_Alone:Anim Phys Sim Space Settings Structure)) -> (Output_Get:Anim Phys Sim Space Settings Structure)
   Set SimSpaceSettingForLowerLock(
    execute=@n7.else
    SimSpaceSettingForLowerLock=Get SimSpaceSettingForLowerLock_General() -> (SimSpaceSettingForLowerLock_General:Anim Phys Sim Space Settings Structure)
   )
   > Set SimSpaceSettingForUpperLock(SimSpaceSettingForUpperLock=Get SimSpaceSettingForUpperLock_General() -> (SimSpaceSettingForUpperLock_General:Anim Phys Sim Space Settings Structure))
   n8 = Sequence!(execute=@n4.then_1) -> (then_0, then_1, then_2:Exec, then_3:Exec, then_4:Exec, then_5:Exec, then_6:Exec)
   n9 = Get Body Mesh() -> (Body Mesh:Scene Component Object Reference)
   n10 = BreakTransform(InTransform=GetSocketTransform(self=@n9, InSocketName=root, TransformSpace=RTS_World))
   Set Root Location(execute=@n8.then_6, Root Location:Vector=@n10) -> (Output_Get:Vector)
   > Set Root Rotation(Root Rotation:Rotator=@n10.Rotation) -> (Output_Get:Rotator)
   n11 = BreakTransform(InTransform=GetSocketTransform(self=@n9, InSocketName=pelvis, TransformSpace=RTS_World))
   Set Pelvis Location(execute=@n8.then_0, Pelvis Location:Vector=@n11) -> (Output_Get:Vector)
   > Set Pelvis Rotation(Pelvis Rotation:Rotator=@n11.Rotation) -> (Output_Get:Rotator)
   n12 = BreakTransform(InTransform=GetSocketTransform(self=@n9, InSocketName=thigh_l, TransformSpace=RTS_World))
   Set Thigh Left Location(execute=@n8.then_1, Thigh Left Location:Vector=@n12) -> (Output_Get:Vector)
   > Set Thigh Left Rotation(Thigh Left Rotation:Rotator=@n12.Rotation) -> (Output_Get:Rotator)
   n13 = BreakTransform(InTransform=GetSocketTransform(self=@n9, InSocketName=thigh_r, TransformSpace=RTS_World))
   Set Thigh Right Location(execute=@n8.then_2, Thigh Right Location:Vector=@n13) -> (Output_Get:Vector)
   > Set Thigh Right Rotation(Thigh Right Rotation:Rotator=@n13.Rotation) -> (Output_Get:Rotator)
   n14 = BreakTransform(InTransform=GetSocketTransform(self=@n9, InSocketName=calf_r, TransformSpace=RTS_World))
   Set Calf Right Location(execute=@n8.then_4, Calf Right Location:Vector=@n14) -> (Output_Get:Vector)
   > Set Calf Right Rotation(Calf Right Rotation:Rotator=@n14.Rotation) -> (Output_Get:Rotator)
   n15 = BreakTransform(InTransform=GetSocketTransform(self=@n9, InSocketName=calf_l, TransformSpace=RTS_World))
   Set Calf Left Location(execute=@n8.then_3, Calf Left Location:Vector=@n15) -> (Output_Get:Vector)
   > Set Calf Left Rotation(Calf Left Rotation:Rotator=@n15.Rotation) -> (Output_Get:Rotator)
   n16 = BreakTransform(InTransform=GetSocketTransform(self=@n9, InSocketName=root, TransformSpace=RTS_World))
   Set Body Root Location(execute=@n8.then_5, Body Root Location:Vector=@n16) -> (Output_Get:Vector)
   > Set Body Root Rotation(Body Root Rotation:Rotator=@n16.Rotation) -> (Output_Get:Rotator)
  ]}
]
variables [
 Body Mesh {type=SceneComponent, default=null}
 Body Root Location {type=Vector}
 Body Root Rotation {type=Rotator}
 Calf Left Location {type=Vector}
 Calf Left Rotation {type=Rotator}
 Calf Right Location {type=Vector}
 Calf Right Rotation {type=Rotator}
 IsPreviewingItemAloneInFN {type=boolean}
 Pelvis Location {type=Vector}
 Pelvis Rotation {type=Rotator}
 RigidBodyAlphaForDisplayAlone {type=number}
 RigidBodyAlphaForGeneral {type=number}
 Root Location {type=Vector}
 Root Rotation {type=Rotator}
 SimSpaceSettingForLowerLock {
  type=AnimPhysSimSpaceSettings
  default {maxAngularAcceleration=10000, maxAngularVelocity=10000, simSpaceAngularAlpha=0, externalAngularVelocity=<0 0 0>}
 }
 SimSpaceSettingForLowerLock_FN_Alone {
  type=AnimPhysSimSpaceSettings
  default {maxAngularAcceleration=15.5, maxAngularVelocity=3.1, simSpaceAngularAlpha=1, externalAngularVelocity=<0 0 0>}
 }
 SimSpaceSettingForLowerLock_General {
  type=AnimPhysSimSpaceSettings
  default {maxAngularAcceleration=22.5, maxAngularVelocity=4.5, simSpaceAngularAlpha=1, externalAngularVelocity=<0 0 0>}
 }
 SimSpaceSettingForUpperLock {
  type=AnimPhysSimSpaceSettings
  default {maxAngularAcceleration=10000, maxAngularVelocity=10000, simSpaceAngularAlpha=0, externalAngularVelocity=<0 0 0>}
 }
 SimSpaceSettingForUpperLock_FN_Alone {
  type=AnimPhysSimSpaceSettings
  default {maxAngularAcceleration=15, maxAngularVelocity=3, simSpaceAngularAlpha=1, externalAngularVelocity=<0 0 0>}
 }
 SimSpaceSettingForUpperLock_General {
  type=AnimPhysSimSpaceSettings
  default {maxAngularAcceleration=2.5, maxAngularVelocity=5, simSpaceAngularAlpha=1, externalAngularVelocity=<0 0 0>}
 }
 Thigh Left Location {type=Vector}
 Thigh Left Rotation {type=Rotator}
 Thigh Right Location {type=Vector}
 Thigh Right Rotation {type=Rotator}
]
